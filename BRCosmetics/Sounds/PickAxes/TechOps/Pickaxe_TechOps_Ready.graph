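# SoundCue for the ready sound of the TechOps pickaxe cosmetic.
# The cue has no root or sound nodes, so it references no audio to play.
# It still carries playback settings: the shared cosmetic pickaxe ready audio property sheet, the melee weapon sound class, a reduced volume multiplier and prime-on-load.
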SoundCue /BRCosmetics/Sounds/PickAxes/TechOps/Pickaxe_TechOps_Ready
root=null
nodes []
properties {
 audioPropertiesSheet=/BRCosmetics/Sounds/Settings/PropertySheets/Pickaxe/APS_Cosmetic_Pickaxe_Ready
 bPrimeOnLoad=true
 maxDistance=2097152
 soundClassObject=&Fort_WeaponMelee
 volumeMultiplier=0.1
}
